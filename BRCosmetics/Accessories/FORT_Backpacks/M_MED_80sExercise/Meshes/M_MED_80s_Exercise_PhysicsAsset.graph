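# Physics asset for the M_MED 80s Exercise backpack, a cosmetic back accessory mesh.
# It holds a single body on backpack_root, a tilted capsule that approximates the pack's volume.
# There are no constraints, so it defines one collision shape rather than a jointed ragdoll or dangling chain.
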PhysicsAsset /BRCosmetics/Accessories/FORT_Backpacks/M_MED_80sExercise/Meshes/M_MED_80s_Exercise_PhysicsAsset
bodies {backpack_root [{shapeType=Capsule, length=28.1166, radius=16.6289, center=<-1.9968 -5.2496 0.7743>, rotation=<rot -2.603 -90 -23.5231>}]}
constraints []
