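# Animation Blueprint for the 80s Exercise backpack cosmetic that procedurally bounces what appear to be its speaker bones.
# Each update, a random stream value scales game time inside a sine, producing an erratic Y-axis speaker_bounce vector.
# The AnimGraph starts from the ref pose, adds a fixed bone-space offset to backpack_root, then applies that bounce additively to drv_child_01 and drv_child_02 in parent-bone space.
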AnimBlueprint /BRCosmetics/Accessories/FORT_Backpacks/M_MED_80sExercise/Meshes/M_MED_80s_Exercise_Skeleton_AnimBP
target_skeleton=&/M_MED_80s_Exercise_Skeleton
graphs [
 AnimGraph {nodes [
   n1 = "Transform (Modify) Bone - Bone: backpack_root"(ComponentPose=Local To Component(LocalPose=Local Space Ref Pose()), Alpha=1, Translation="1,-4,0", Rotation="0,0,0", Scale="1,1,1") {
    translationMode=BMM_Additive
    translationSpace=BCS_BoneSpace
    alphaBoolBlend {blendOption=Linear}
    boneToModify {boneName=backpack_root}
    translation=<1 -4>
   }
   n2 = Get speaker_bounce() -> (speaker_bounce:Vector)
   n3 = "Transform (Modify) Bone - Bone: drv_child_01"(ComponentPose=@n1, Alpha=1, Translation=@n2, Rotation="0,0,0", Scale="1,1,1") {
    translationMode=BMM_Additive
    translationSpace=BCS_ParentBoneSpace
    alphaBoolBlend {blendOption=Linear}
    boneToModify {boneName=drv_child_01}
   }
   n4 = "Transform (Modify) Bone - Bone: drv_child_02"(ComponentPose=@n3, Alpha=1, Translation=@n2, Rotation="0,0,0", Scale="1,1,1") {
    translationMode=BMM_Additive
    translationSpace=BCS_ParentBoneSpace
    alphaBoolBlend {blendOption=Linear}
    boneToModify {boneName=drv_child_02}
   }
   Output Pose(Result=Component To Local(ComponentPose=@n4)) {layerGroup=None}
  ]}
 EventGraph {nodes [
   Event BlueprintUpdateAnimation()
   n1 = RandomFloatFromStream!(Stream:"Random Stream Structure (by ref)"=Get random_stream() -> (random_stream:Random Stream Structure)) -> (ReturnValue:float)
   n2 = "Sin (Radians)"(A=float * float(A=float * float(A=@n1, B=0.333), B=GetGameTimeInSeconds()))
   > Set speaker_bounce(speaker_bounce:Vector=MakeVector(Y=float * float(A=@n2, B=1.123))) -> (Output_Get:Vector)
  ]}
]
variables [
 random_stream {type=RandomStream, default {initialSeed=0, seed=0}}
 speaker_bounce {type=Vector}
]
